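# Translucent, mesh-particle-capable material for the hero shard in the Hero Rescue mission, animating warm-colored shards that spin and bob.
# Pivot Painter per-object data gives each shard random speed and direction, with a static switch choosing between painted axes and a parameter axis with vertical bob.
# A normal-corrected Fresnel rim adds to a sine-pulsed orange glow varied by a foam texture.
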Material /Game/Missions/Secondary/HeroRescue/M_Shard_HeroVar
graph {
 properties {
  bUsedWithMeshParticles=true
  blendMode=BLEND_Translucent
  pixelDepthOffsetMode=PDOM_Legacy
  previewMesh=&/S_HeroShard
  usedShadingModels=Default Lit
 }
 nodes [
  n1 = Constant3Vector() {constant=<rgba 1.25 0.6457 0.3688 0>}
  n2 = MaterialFunctionCall() {materialFunction=/Engine/Functions/Engine_MaterialFunctions02/PivotPainter/PivotPainter_PerObjectData}
  n3 = StaticSwitchParameter(True=@n2.X-Axis Vector, False=VectorParameter() {parameterName=Rotation Axis, defaultValue=<rgba 1 0 0 0>}.RGB) {defaultValue=true, parameterName=PivotPainterRotation}
  n4 = Multiply(
   A=Multiply(A=Time(), B=LinearInterpolate(Alpha=@n2.Random Value Per Element) {constA=-1})
   B=ScalarParameter() {parameterName=Rotation Speed}
  ) {constB=0.4}
  n5 = Fresnel(Normal=MaterialFunctionCall(Rotation Axis=@n3, Rotation Angle=@n4) {materialFunction=&FixRotateAboutAxisNormals}.New World Space Vertex Normal) {baseReflectFraction=0, exponent=8}
  n6 = Constant3Vector() {constant=<rgba 0.795 0.2514 0.1215 0>}
  n7 = Constant3Vector() {constant=<rgba 0.775 0.3237 0.0423 0>}
  n8 = TextureSample(UVs=Rotator(Coordinate=TextureCoordinate() {uTiling=0.5, vTiling=0.5}) {centerX=0, centerY=0, speed=0}) {texture=/Game/Effects/Fort_Cine_Effects/Leaves/Textures/T_sea_foam_03}
  n9 = Sine(None=Time()) {period=2}
  Add(
   A=Multiply(A=Multiply(A=@n1, B=@n5), B=Constant() {r=2})
   B=Multiply(A=LinearInterpolate(A=@n6, B=@n7, Alpha=@n8.B), B=ConstantBiasScale(None=@n9))
  )
  Multiply(A=Constant() {r=0.75}, B=Constant() {r=0.75})
  RotateAboutAxis(NormalizedRotationAxis=@n3, RotationAngle=@n4, PivotPoint=ObjectPositionWS(), Position=WorldPosition().XYZ)
  n10 = Sine(None=Time()) {period=3}
  n11 = Multiply(A=LinearInterpolate(Alpha=@n2.Random Value Per Element) {constA=0.25}, B=@n10)
  n12 = ScalarParameter() {defaultValue=7, parameterName=Height bob}
  n13 = Multiply(A=@n10, B=LinearInterpolate(Alpha=@n2.Random Value Per Element) {constA=-1})
  StaticSwitchParameter(
   True=Multiply(A=@n2.X-Axis Vector, B=Multiply(A=@n11, B=@n12))
   False=Multiply(A=Multiply(A=@n12, B=@n13), B=Constant3Vector() {constant=<rgba 0 0 1 0>})
  ) {defaultValue=true, parameterName=PivotPainterRotation}
 ]
}
